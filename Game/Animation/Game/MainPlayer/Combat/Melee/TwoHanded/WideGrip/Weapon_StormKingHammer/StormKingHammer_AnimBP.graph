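# Animation Blueprint for the Storm King Hammer melee weapon, switching the weapon mesh between its idle animation and a fist pose.
# Each update reads the MeleeAnim curve from the skeletal mesh anim instance on the weapon actor's owner to set MeleeAnimPlaying.
# That bool selects the fist pose while a melee animation plays and idle otherwise.
# The blended result passes through DefaultSlot, allowing montages to override it.
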AnimBlueprint /Game/Animation/Game/MainPlayer/Combat/Melee/TwoHanded/WideGrip/Weapon_StormKingHammer/StormKingHammer_AnimBP
target_skeleton=/Game/Weapons/FORT_Melee/Meshes/SK_Storm_King_Hammer_Skeleton
graphs [
 AnimGraph {nodes [
   n1 = Sequence Player 'StormKingHammer_Idle_Weapon'() {sequence=&/StormKingHammer_Idle_Weapon}
   n2 = Sequence Player 'StormKingHammer_Fist_Pose_Weapon'() {sequence=&/StormKingHammer_Fist_Pose_Weapon}
   n3 = Get MeleeAnimPlaying() -> (MeleeAnimPlaying:bool)
   Output Pose(Result=Slot 'DefaultSlot'(Source=Blend Poses by bool(BlendPose_0=@n2, BlendPose_1=@n1, BlendTime_1=0.2, bActiveValue=@n3))) {layerGroup=None}
  ]}
 EventGraph {nodes [
   Event BlueprintUpdateAnimation()
   > n1 = Cast To B_Melee_Blunt_Hammer_Stormking!(execute, Object=GetOwningActor()) -> (then, CastFailed, AsB Melee Blunt Hammer Stormking:B Melee Blunt Hammer Stormking Object Reference) {is=title K2Node_DynamicCast}
   n2 = GetOwner(self:Actor Object Reference=@n1)
   > Is Valid(InputObject=@n2)
   n3 = GetComponentByClass!(self:Actor Object Reference=@n2, ComponentClass:Actor Component Class Reference=/Script/Engine.SkeletalMeshComponent) -> (ReturnValue:Skeletal Mesh Component Object Reference)
   > Is Valid(InputObject=@n3)
   n4 = GetAnimInstance(self=@n3)
   > Set MeleeAnimPlaying(MeleeAnimPlaying:bool="float >= float"(A=GetCurveValue(self=@n4, CurveName=MeleeAnim), B=1) {is=title K2Node_CallFunction}) -> (Output_Get:bool)
  ]}
]
variables [MeleeAnimPlaying {type=boolean}]
